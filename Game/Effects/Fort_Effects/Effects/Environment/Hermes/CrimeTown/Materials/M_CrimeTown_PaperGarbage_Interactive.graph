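# Masked, two-sided lit material for CrimeTown paper garbage on Niagara mesh particles, using GarbageIsland trash textures.
# It darkens the trash diffuse texture and samples matching normal and specular maps.
# A panning Gaussian noise texture, scaled by the WorldWPO dynamic parameter and object position, forms what appears to be a world position offset.
# That parameter defaults to zero, so particles must supply a value for the motion term to take effect.
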Material /Game/Effects/Fort_Effects/Effects/Environment/Hermes/CrimeTown/Materials/M_CrimeTown_PaperGarbage_Interactive
graph {
 properties {
  bAutomaticallySetUsageInEditor=false
  bUsedWithNiagaraMeshParticles=true
  blendMode=BLEND_Masked
  twoSided=true
  usedShadingModels=Default Lit
 }
 nodes [
  Constant()
  Multiply(A=Constant() {r=0.7}, B=TextureSample() {texture=/Game/Environments/Apollo/Sets/GarbageIsland/Props/Textures/T_GarbageIsland_Trash_D}.RGB) {constA=0.7}
  n1 = TextureSample(UVs=Panner(Coordinate=TextureCoordinate()) {speedX=0.3, speedY=0.2}) {texture=/Game/Effects/Fort_Effects/Textures/Tile/T_GaussianNoise_01_M}
  Multiply(
   A=Divide(A=ObjectPositionWS()) {constB=10}
   B=Multiply(A=Multiply(A=@n1.RGB) {constB=0.5}, B=DynamicParameter() {defaultValue=<rgba 0 1 1 1>, paramNames [WorldWPO,Param2,Param3,Param4]}.WorldWPO) {constB=0}
  )
  TextureSample() {samplerType=SAMPLERTYPE_Normal, texture=/Game/Environments/Apollo/Sets/GarbageIsland/Props/Textures/T_GarbageIsland_Trash_N}
  TextureSample() {texture=/Game/Environments/Apollo/Sets/GarbageIsland/Props/Textures/T_GarbageIsland_Trash_S}
 ]
}
