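# GPU Niagara system of crawling mesh bugs for CrimeTown environments that scurry away from the player's camera.
# Bugs spawn in a user-scalable volume, snap to distance field surfaces and are filtered by GBuffer hue.
# Each update, negative point attraction at the camera and region-masked curl noise drive surface-snapped movement.
# A timed state machine makes motion stop-and-go, while a second region mask shrinks and kills bugs straying from the system.
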NiagaraSystem /Game/Effects/Fort_Effects/Effects/Environment/Hermes/CrimeTown/NS_CrimeTown_Bugs_AvoidPlayer
dependencies {
 usedDataInterfaces [
  &NiagaraDataInterfaceCamera
  &NiagaraDataInterfaceCurve
  &NiagaraDataInterfaceCollisionQuery
  &NiagaraDataInterfaceGBuffer
 ]
 usedDynamicInputs []
 usedModules [&SystemState]
}
emitters {Bugs_AvoidPlayer {
  simTarget=GPUComputeSim
  emitterSpawnScript [/SetVariables() {isSetParameters=true}]
  emitterUpdateScript [
   &EmitterState(
    Life Cycle Mode=::NewEnumerator1
    Loop Behavior=::NewEnumerator1
    Loop Duration Mode=::NewEnumerator1
    Reset Age On Awaken=true
    Scalability Mode=::NewEnumerator1
    UseLoopDelay=false
   )
   &SpawnRate(
    SpawnRate=&FloatFromCurve(
     FloatCurve=DataInterface() {propertyValues {Curve {keys [<at 0 0.3>,<at 0.5 1>,<at 1 0>]}}}
     CurveIndex=$Emitter.Age
     Scale Curve=&Multiply_Float(A=$User.Bugs_SpawnRateMultiplier, B=200)
    )
    Use Spawn Probability=false
   )
  ]
  particleSpawnScript [
   &InitializeParticle(
    Color Maximum=<rgba 0.443 0.2919 0.0125 1>
    Color Minimum=<rgba 0.0256 0.0256 0.0256 1>
    Color Mode=::NewEnumerator2
    Lifetime=5
    Mesh Scale Mode=::NewEnumerator1
    Mesh Uniform Scale Max=0.5
    Mesh Uniform Scale Min=0.35
    UsePositionOffset=false
    Write Material Random=true
   )
   &ShapeLocation(
    MaskSpawnGroup=false
    Non Uniform Scale=$User.TransformScale
    Override Seed=false
    Rotation Mode=::NewEnumerator2
    Sphere Radius=&Multiply_Float(A=100, B=$User.RadiusMult)
    UseOwnerScale=true
    Yaw / Pitch / Roll=<0 0.25 0>
   ) {module=ShapeLocation001}
   /Niagara/Modules/Math/MoveToNearestDistanceFieldSurface_GPU(Kill by Distance=2500, Offset Particle By Radius=true, Radius=3, bKillByDistance=true)
   /Game/Effects/Niagara/Modules/Spawn/FN_Spawn_GBuffer_Color_WithinRangeV2(HueMax=60, HueMin=44)
   /SetVariables(Particles.InitialPos:NiagaraPosition=$Particles.Position) {isSetParameters=true}
   &InitialMeshOrientation(
    Mesh Orientation Mode=::NewEnumerator2
    Rotation=&RandomRangeVector(Minimum=<0 0 1>, Maximum=<0 0 -1>)
    Use Rotation Vector=true
   )
   &InitialMeshOrientation(
    Mesh Orientation Mode=::NewEnumerator2
    Orientation Vector=$Output.MoveToNearestDistanceFieldSurface_GPU.VectorToNearestSurfaceNormal
    Use Rotation Vector=false
   ) {module=InitialMeshOrientation001}
  ]
  particleUpdateScript [
   &ParticleState(Kill Particles When Lifetime Has Elapsed=false)
   /Niagara/Modules/Update/StateChange/TimeBasedStateMachine(
    Lerp Time To ON:NiagaraFloat=&RandomRangeFloat(Minimum=0.5, Maximum=1)
    Lerp Time To Off:NiagaraFloat=&RandomRangeFloat(Minimum=0.5, Maximum=1)
    Off Duration:NiagaraFloat=&RandomRangeFloat(Minimum=0.3, Maximum=0.7)
    On Duration:NiagaraFloat=&RandomRangeFloat(Minimum=0.5, Maximum=1.1)
   )
   &PointAttractionForce(
    Attraction Radius=500
    Attraction Strength=&RandomRangeFloat(Minimum=-100, Maximum=-200)
    Attractor Position:NiagaraPosition=&ReturnCameraPosition(Coordinate Space:ENiagaraCoordinateSpace=::World, Camera Query:NiagaraDataInterfaceCamera=DataInterface())
    Falloff Exponent:NiagaraFloat=&RandomRangeFloat(Minimum=0.35, Maximum=0.5)
    Kill Within Radius=false
    Module.Use Falloff=true
    UseAttractorOffset=false
   )
   &RegionMask(
    Region Origin:Vector3f=&ReturnCameraPosition(Coordinate Space:ENiagaraCoordinateSpace=::World, Camera Query:NiagaraDataInterfaceCamera=DataInterface())
    Region Radius=500
    UseOffset=false
    UseRotation=false
   ) {module=RegionMask001}
   &CurlNoiseForce(
    Bias Noise=false
    Noise Frequency=20
    Noise Strength=&MakeCustomFloatFromBool(
     Bool=&SetBoolByFloatComparison(A=$Particles.RegionMask001.Float, B=0, Comparison Type=::NewEnumerator7)
     True Float=6517.1602
     False Float=160
    )
    Pan Noise=false
   ) {enabled=false}
   &CurlNoiseForce(
    Bias Noise=false
    Noise Frequency=20
    Noise Strength=&MakeCustomFloatFromBool(
     Bool=&SetBoolByFloatComparison(A=$Particles.RegionMask001.Float, B=0, Comparison Type=::NewEnumerator7)
     True Float=160
     False Float=&RandomRangeFloat(Minimum=1200, Maximum=2800)
    )
    Pan Noise=false
   ) {module=CurlNoiseForce001}
   &MoveToNearestDistanceFieldSurface_GPU(Kill by Distance=8000, Offset Particle By Radius=true, bKillByDistance=true) {module=MoveToNearestDistanceFieldSurface_GPU001}
   &SolveForcesAndVelocity(Acceleration Limit=150, Clamp Velocity=true, Limit Acceleration=true, Speed Limit=150)
   &RegionMask(Invert Region=true, Region Radius=850, UseOffset=false, UseRotation=false)
   &KillParticles(Kill Particles=$Particles.RegionMask.IsInsideMask)
   &ScaleMeshSize(Scale Factor=&Multiply_Vector(
     A=&VectorFromFloat(Value=&FloatFromCurve(
       FloatCurve=DataInterface() {propertyValues {Curve {keys [{interpMode=RCIM_Cubic, tangentMode=RCTM_User, time=0, value=0}, {interpMode=RCIM_Cubic, tangentMode=RCTM_User, time=0.5, value=1}]}}}
       CurveIndex=$Emitter.Age
      ))
     B=&VectorFromFloat(Value=&FloatFromCurve(
       FloatCurve=DataInterface() {propertyValues {Curve {keys [{interpMode=RCIM_Cubic, tangentMode=RCTM_User, time=0.8, value=1}, {interpMode=RCIM_Cubic, tangentMode=RCTM_User, time=1, value=0}]}}}
       CurveIndex=$Particles.RegionMask.Float
      ))
    ))
   &self:BugsCrawling(Position:NiagaraPosition=$Particles.Position, Velocity:Vector3f=$Particles.Velocity) {module=BugsCrawling}
   &UpdateMeshOrientation(Orientation Method=::NewEnumerator3) {module=UpdateMeshOrientation001}
   /SetVariables(Particles.Velocity=&Multiply_VectorByFloat(Vector=$Particles.Velocity, Float=$Output.TimeBasedStateMachine.On Off Percentage)) {isSetParameters=true}
  ]
  rendererClasses [&NiagaraMeshRendererProperties]
  renderers [{rendererClass=&NiagaraMeshRendererProperties, rendererIndex=0}]
 }}
summary {
 systemName=NS_CrimeTown_Bugs_AvoidPlayer
 emitters [{emitterName=Bugs_AvoidPlayer, simTarget=GPUComputeSim, rendererClasses [&NiagaraMeshRendererProperties]}]
 userVariables [
  User.Bugs_SpawnRateMultiplier {type=NiagaraFloat, defaultValue=1, description=""}
  User.RadiusMult {type=NiagaraFloat, defaultValue=1, description=""}
  User.TransformScale {type=Vector3f, description="", defaultValue {struct=Vector3f, value=<1 1 1>}}
 ]
}
